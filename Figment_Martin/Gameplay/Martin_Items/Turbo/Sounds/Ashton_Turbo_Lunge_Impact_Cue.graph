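# Sound cue for the Turbo item's lunge impact, mixing a close layer with a distant layer.
# The close layer randomly picks a ground pound wave, varies its pitch and uses close weapon attenuation.
# The half-volume distant layer reuses BRSword distant impacts, stacks two pitch modulators for a lower pitch, and uses far attenuation and a distant-fire sound class.
# A concurrency override allows one instance per owner, stopping the oldest.
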SoundCue /Figment_Martin/Gameplay/Martin_Items/Turbo/Sounds/Ashton_Turbo_Lunge_Impact_Cue
root=Mixer_1
nodes [
 {id=Attenuation_1, children [Modulator_0], properties {attenuationSettings=/Game/Sounds/Attenuation/Fort_Weapon_Carmine_Close_Athena_Att}}
 {id=Attenuation_2, children [Modulator_1], properties {attenuationSettings=/Game/Sounds/Attenuation/Fort_Weapon_Carmine_Far_Athena_Att}}
 {id=Mixer_1, children [SoundClass_1,SoundClass_2], properties {inputVolume [1,0.5]}}
 {id=Modulator_0, children [Random_4], properties {pitchMax=1.2, pitchMin=1}}
 {id=Modulator_1, children [Modulator_3], properties {pitchMax=1.2, pitchMin=1}}
 {id=Modulator_3, children [Random_3], properties {pitchMax=0.8, pitchMin=0.7}}
 {id=Random_3, children [WavePlayer_18, WavePlayer_19, WavePlayer_20, WavePlayer_21], properties {weights [1,1,1,1]}}
 {id=Random_4, children [WavePlayer_8, WavePlayer_9, WavePlayer_10, WavePlayer_11], properties {weights [1,1,1,1]}}
 {id=SoundClass_1, children [Attenuation_1]}
 {
  id=SoundClass_2
  children [Attenuation_2]
  properties {soundClassOverride=&Fort_WeaponRanged_Fire_Distant}
 }
 {id=WavePlayer_10, properties {soundWaveAssetPtr=&/Ashton_Turbo_Lunge_Ground_Pound_02}}
 {id=WavePlayer_11, properties {soundWaveAssetPtr=&/Ashton_Turbo_Lunge_Ground_Pound_03}}
 {id=WavePlayer_18, properties {soundWaveAssetPtr=/Game/Sounds/Fort_GamePlay_Sounds/BRSword/AB_Impact_01_distant}}
 {id=WavePlayer_19, properties {soundWaveAssetPtr=/Game/Sounds/Fort_GamePlay_Sounds/BRSword/AB_Impact_02_distant}}
 {id=WavePlayer_20, properties {soundWaveAssetPtr=/Game/Sounds/Fort_GamePlay_Sounds/BRSword/AB_Impact_03_distant}}
 {id=WavePlayer_21, properties {soundWaveAssetPtr=/Game/Sounds/Fort_GamePlay_Sounds/BRSword/AB_Impact_04_distant}}
 {id=WavePlayer_8, properties {soundWaveAssetPtr=&/Ashton_Turbo_Lunge_Ground_Pound_04}}
 {id=WavePlayer_9, properties {soundWaveAssetPtr=&/Ashton_Turbo_Lunge_Ground_Pound_01}}
]
properties {
 bOverrideConcurrency=true
 maxDistance=25000
 volumeMultiplier=1
 concurrencyOverrides {bLimitToOwner=true, maxCount=1, resolutionRule=StopOldest}
}
