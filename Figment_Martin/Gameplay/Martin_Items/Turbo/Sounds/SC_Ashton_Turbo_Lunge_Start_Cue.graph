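# Sound cue for the start of the Turbo item's lunge in Figment_Martin, using Ashton-named waves.
# A random node picks one of three equally weighted wave variations and feeds a sound class node that overrides the class to third-person ranged weapon fire.
# Playback uses shared Ashton ability attenuation, and a concurrency override allows one instance per owner, stopping the oldest.
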SoundCue /Figment_Martin/Gameplay/Martin_Items/Turbo/Sounds/SC_Ashton_Turbo_Lunge_Start_Cue
root=SoundClass_1
nodes [
 {id=Random_0, children [WavePlayer_0,WavePlayer_1,WavePlayer_2], properties {weights [1,1,1]}}
 {id=SoundClass_1, children [Random_0], properties {soundClassOverride=&Fort_WeaponRanged_Fire_3P_PSM}}
 {id=WavePlayer_0, properties {soundWaveAssetPtr=&/Ashton_Turbo_Lunge_Start_01}}
 {id=WavePlayer_1, properties {soundWaveAssetPtr=&/Ashton_Turbo_Lunge_Start_02}}
 {id=WavePlayer_2, properties {soundWaveAssetPtr=&/Ashton_Turbo_Lunge_Start_03}}
]
properties {
 attenuationSettings=/Game/Sounds/Attenuation/Fort_Ability_Ashton_Gen
 bOverrideConcurrency=true
 maxDistance=9300
 concurrencyOverrides {bLimitToOwner=true, maxCount=1, resolutionRule=StopOldest}
}
